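# Sound cue for the looping base sound in Creative's Capture the Flag mode, routed to the Fort_Gadgets_LTM sound class.
# A single looping wave player plays CTF_Base_Loop_01 at reduced volume.
# Overridden attenuation uses natural-sound falloff from a zero-size inner shape over a short range.
# Interpolated listener focus makes the loop quieter when it lies outside the listener's narrow focus cone.
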SoundCue /Game/Sounds/Creative/Modes/CTF/CTF_Base_Loop_Cue
root=WavePlayer_0
nodes [{id=WavePlayer_0, properties {bLooping=true, soundWaveAssetPtr=&/CTF_Base_Loop_01}}]
properties {
 bOverrideAttenuation=true
 maxDistance=1000
 soundClassObject=/Game/Sounds/FortSoundClassesAndModes/Fort_Gadgets_LTM
 volumeMultiplier=0.3
 attenuationOverrides {
  bEnableFocusInterpolation=true
  bEnableListenerFocus=true
  distanceAlgorithm=NaturalSound
  falloffDistance=1000
  focusAzimuth=15
  nonFocusAzimuth=45
  nonFocusVolumeAttenuation=0.5
  attenuationShapeExtents {x=0}
 }
}
