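# Sound Cue for Creative Capture the Flag, played, going by its name, when the enemy delivers a flag.
# A single Wave Player feeds the root directly, with no randomization or mixing.
# Overridden attenuation disables spatialization while keeping distance falloff, so it fades with range but is not panned.
# It runs at reduced volume through the LTM in-game master sound class and shares a CTF delivery concurrency asset.
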SoundCue /Game/Sounds/Creative/Modes/CTF/CTF_Deliver_Enemy_Cue
root=WavePlayer_2
nodes [{id=WavePlayer_2, properties {soundWaveAssetPtr=&/CTF_Deliver_Enemy_01}}]
properties {
 bOverrideAttenuation=true
 maxDistance=9000
 priority=0.9
 soundClassObject=/Game/Sounds/FortSoundClassesAndModes/Fort_LTM_BR_InGame_Master
 volumeMultiplier=0.5
 attenuationOverrides {bSpatialize=false, falloffDistance=6000, attenuationShapeExtents {x=3000}}
 concurrencyOverrides {resolutionRule=PreventNew}
 concurrencySet [/Game/Sounds/Creative/Concurrency/Creative_CTF_Delivery_Conc]
}
