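# Sound cue for a flag delivery in Creative Capture the Flag, apparently the team variant.
# A Mixer layers the base delivery wave at reduced volume under the team-specific wave at full volume.
# Overridden attenuation fades it over distance without spatialization, and it routes through the LTM BR in-game master sound class at half volume.
# It references a shared CTF delivery concurrency set, with a PreventNew rule in its concurrency overrides.
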SoundCue /Game/Sounds/Creative/Modes/CTF/CTF_Deliver_Team_Cue
root=Mixer_0
nodes [
 {id=Mixer_0, children [WavePlayer_1,WavePlayer_0], properties {inputVolume [0.6,1]}}
 {id=WavePlayer_0, properties {soundWaveAssetPtr=&/CTF_Deliver_Team_01}}
 {id=WavePlayer_1, properties {soundWaveAssetPtr=&/CTF_Deliver_01}}
]
properties {
 bOverrideAttenuation=true
 maxDistance=9000
 soundClassObject=/Game/Sounds/FortSoundClassesAndModes/Fort_LTM_BR_InGame_Master
 volumeMultiplier=0.5
 attenuationOverrides {bSpatialize=false, falloffDistance=6000, attenuationShapeExtents {x=3000}}
 concurrencyOverrides {resolutionRule=PreventNew}
 concurrencySet [/Game/Sounds/Creative/Concurrency/Creative_CTF_Delivery_Conc]
}
