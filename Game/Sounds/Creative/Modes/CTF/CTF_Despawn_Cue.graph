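# Sound cue for the despawn sound in Creative Capture the Flag (CTF) mode.
# A single wave player plays CTF_Despawn_01 at slightly reduced volume, with no randomization or layering.
# It uses the Creative CTF attenuation asset, the LTM Battle Royale in-game master sound class and the CTF delivery concurrency set.
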SoundCue /Game/Sounds/Creative/Modes/CTF/CTF_Despawn_Cue
root=WavePlayer_1
nodes [{id=WavePlayer_1, properties {soundWaveAssetPtr=&/CTF_Despawn_01}}]
properties {
 attenuationSettings=/Game/Sounds/Creative/Attenuations/Fort_Creative_CTF_Att
 maxDistance=20000
 priority=0.5
 soundClassObject=/Game/Sounds/FortSoundClassesAndModes/Fort_LTM_BR_InGame_Master
 volumeMultiplier=0.8
 attenuationOverrides {falloffDistance=20000, attenuationShapeExtents {x=0}}
 concurrencyOverrides {resolutionRule=PreventNew}
 concurrencySet [/Game/Sounds/Creative/Concurrency/Creative_CTF_Delivery_Conc]
}
