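# Sound cue for the looping flag sound in Creative's Capture the Flag mode.
# A looping wave player feeds a continuous modulator, which drives both volume and pitch from a Speed parameter that defaults to zero.
# It plays in the gadgets LTM sound class at reduced volume, using small-toy attenuation overridden with natural-sound falloff and occlusion traces.
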SoundCue /Game/Sounds/Creative/Modes/CTF/CTF_FlagLoop_Cue
root=ModulatorContinuous_0
nodes [
 {
  id=ModulatorContinuous_0
  children [WavePlayer_0]
  properties {pitchModulationParams {default=0, minInput=0.9, minOutput=0.9, parameterName=Speed}, volumeModulationParams {default=0, parameterName=Speed}}
 }
 {id=WavePlayer_0, properties {bLooping=true, soundWaveAssetPtr=&/CTF_FlagLoop_01}}
]
properties {
 attenuationSettings=/Game/Sounds/Creative/Attenuations/Fort_Creative_SmallToy_Att
 bOverrideAttenuation=true
 maxDistance=1900
 soundClassObject=/Game/Sounds/FortSoundClassesAndModes/Fort_Gadgets_LTM
 volumeMultiplier=0.5
 attenuationOverrides {bEnableOcclusion=true, distanceAlgorithm=NaturalSound, falloffDistance=1500, occlusionTraceChannel=ECC_GameTraceChannel11}
}
