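# Sound cue for the looping audio of a dropped flag in Creative's Capture the Flag mode.
# A single looping wave player plays CTF_FlagLoop_01 at reduced volume through the Fort_Gadgets_LTM sound class.
# Inline attenuation overrides the assigned SmallToy preset, using natural-sound distance falloff, listener focus that attenuates off-focus playback, and trace-based occlusion.
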SoundCue /Game/Sounds/Creative/Modes/CTF/CTF_FlagLoop_Dropped_Cue
root=WavePlayer_0
nodes [{id=WavePlayer_0, properties {bLooping=true, soundWaveAssetPtr=&/CTF_FlagLoop_01}}]
properties {
 attenuationSettings=/Game/Sounds/Creative/Attenuations/Fort_Creative_SmallToy_Att
 bOverrideAttenuation=true
 maxDistance=2250
 soundClassObject=/Game/Sounds/FortSoundClassesAndModes/Fort_Gadgets_LTM
 volumeMultiplier=0.3
 attenuationOverrides {
  bEnableFocusInterpolation=true
  bEnableListenerFocus=true
  bEnableOcclusion=true
  dBAttenuationAtMax=-40
  distanceAlgorithm=NaturalSound
  falloffDistance=1500
  focusAzimuth=15
  nonFocusAzimuth=45
  nonFocusVolumeAttenuation=0.4
  nonSpatializedRadiusStart=100
  occlusionTraceChannel=ECC_GameTraceChannel11
  occlusionVolumeAttenuation=0.7
  stereoSpread=2000
  attenuationShapeExtents {x=0}
 }
}
