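# Sound cue for the object-added sound in the Creative Capture the Flag mode.
# A single wave player plays the reused ItemSpawner object-added-to-spawner wave at reduced volume.
# It uses the Creative small-toy attenuation and the LTM Battle Royale in-game master sound class.
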SoundCue /Game/Sounds/Creative/Modes/CTF/CTF_ObjectAdded_Cue
root=WavePlayer_0
nodes [{id=WavePlayer_0, properties {soundWaveAssetPtr=/Game/Sounds/Creative/Toys/ItemSpawner/ItemSpawner_ObjectAddedToSpawner}}]
properties {
 attenuationSettings=/Game/Sounds/Creative/Attenuations/Fort_Creative_SmallToy_Att
 maxDistance=3000
 soundClassObject=/Game/Sounds/FortSoundClassesAndModes/Fort_LTM_BR_InGame_Master
 volumeMultiplier=0.6
}
