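# Sound cue for a looping zone sound in the Creative Capture the Flag (CTF) mode, routed to the Gadgets LTM sound class at low volume.
# A looping wave player feeds a modulator that randomizes pitch within a narrow range well above the original.
# Overridden attenuation uses natural-sound falloff with distance-based low-pass filtering, wide stereo spread and listener focus that quiets the loop when off-focus.
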SoundCue /Game/Sounds/Creative/Modes/CTF/CTF_ZoneLoop_Cue
root=Modulator_0
nodes [
 {id=Modulator_0, children [WavePlayer_0], properties {pitchMax=1.8, pitchMin=1.7}}
 {id=WavePlayer_0, properties {bLooping=true, soundWaveAssetPtr=&/CTF_ZoneLoop_01}}
]
properties {
 bOverrideAttenuation=true
 maxDistance=1200
 soundClassObject=/Game/Sounds/FortSoundClassesAndModes/Fort_Gadgets_LTM
 volumeMultiplier=0.1
 attenuationOverrides {
  bAttenuateWithLPF=true
  bEnableFocusInterpolation=true
  bEnableListenerFocus=true
  distanceAlgorithm=NaturalSound
  falloffDistance=1200
  focusAzimuth=15
  lPFFrequencyAtMax=200
  lPFRadiusMax=1200
  lPFRadiusMin=400
  nonFocusAzimuth=45
  nonFocusVolumeAttenuation=0.5
  stereoSpread=1200
  attenuationShapeExtents {x=0}
 }
}
